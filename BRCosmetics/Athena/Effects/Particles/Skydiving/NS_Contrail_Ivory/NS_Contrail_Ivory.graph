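# Niagara skydiving contrail cosmetic that streams two piano keyboards from the character, with blue ribbons, glows, radial energy, note sprites and a paper burst.
# Left and Right emitters map twelve partitions to key shapes in black or white, randomly flashing some cyan.
# Skeletal-mesh sampling anchors most emitters, with curve-driven, often mirrored offsets blended by ParaGlideLeanAlpha.
# User.Hitting brightens the radial energy; the continuous Paper_Rate emitter is disabled.
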NiagaraSystem /BRCosmetics/Athena/Effects/Particles/Skydiving/NS_Contrail_Ivory/NS_Contrail_Ivory
dependencies {
 usedDataInterfaces [
  &NiagaraDataInterfaceArrayColor
  &NiagaraDataInterfaceCurve
  &NiagaraDataInterfaceSpriteRendererInfo
  &NiagaraDataInterfaceColorCurve
  &NiagaraDataInterfaceSkeletalMesh
 ]
 usedDynamicInputs []
 usedModules [
  &SystemState
  &Timeline
  &BlendSpaceHelper
  &self:SystemSpawnScript.SystemScriptSource.SystemScriptGraph.NiagaraNodeAssignment_0.SetVariables_#1
 ]
}
emitters {
 Init_Glow {
  emitterUpdateScript [
   &EmitterState(Life Cycle Mode=::NewEnumerator1, Loop Behavior=::NewEnumerator1, Loop Duration Mode=::NewEnumerator1, UseLoopDelay=true)
   &SpawnRate(SpawnRate=&Lerp_Float(A=3, B=6, Alpha=$User.ParaGlideLeanAlpha), Use Spawn Probability=false)
  ]
  particleSpawnScript [
   &InitializeParticle(
    Color=<rgba 1.3976 6.8947 10 2>
    Lifetime=&RandomRangeFloat(Minimum=0.5, Maximum=1)
    Mesh Scale Mode=::NewEnumerator3
    Mesh Uniform Scale=&Multiply_Float(A=0.016, B=2)
    UsePositionOffset=true
    Write Material Random=true
   ) {module=InitializeParticle001}
   &PartitionParticles()
   &InitialMeshOrientation(
    Rotation=&RandomRangeVector(Minimum=<0 0 0>, Maximum=<1 1 1>)
    Use Orientation Vector=true
    Use Rotation Vector=true
   ) {module=InitialMeshOrientation001}
   &SystemLocation(MaskSpawnGroup=false, Offset=<0 25 40>) {module=SystemLocation001}
  ]
  particleUpdateScript [
   &ParticleState()
   &SkeletalMeshLocation(
    Bone Sampling Mode=::NewEnumerator4
    MaskSpawnGroup=false
    Orientation Sampling=::NewEnumerator1
    Sampled Position Offset=&Multiply_Vector(
     A=&Add_Vector(
      A=&Lerp_Vector(
       A=&MakeVector(
        X=0
        Y=&FloatFromCurve(FloatCurve=DataInterface() {propertyValues {Curve {defaultValue=1, keys [<at 0 0.32 RCIM_Cubic>]}}}, Scale Curve=50)
        Z=&FloatFromCurve(FloatCurve=DataInterface() {propertyValues {Curve {defaultValue=1, keys [<at 0 -0.23 RCIM_Cubic>]}}}, Scale Curve=100)
       )
       B=&MakeVector(
        X=&FloatFromCurve(FloatCurve=DataInterface() {propertyValues {Curve {defaultValue=1, keys [{tangentMode=RCTM_User, time=0, value=0.3}]}}}, Scale Curve=150)
        Y=&FloatFromCurve(FloatCurve=DataInterface() {propertyValues {Curve {defaultValue=1, keys [<at 0 0.3 RCIM_Cubic>]}}}, Scale Curve=50)
        Z=&FloatFromCurve(FloatCurve=DataInterface() {propertyValues {Curve {defaultValue=1, keys [<at 0 -0.3 RCIM_Cubic>]}}}, Scale Curve=100)
       )
       Alpha=$User.ParaGlideLeanAlpha
      )
      B=<0 -15 -15>
     )
     B=&MergeHelper_VectorSwitch(0=<1 1 1>, 1=<-1 1 1>, 2=?, 3=?, Index=$Particles.PartitionParticles.Partition, Num=::NewEnumerator0)
    )
    Sampled Position Offset Coordinate Space:ENiagaraSkelMeshTransforms=::NewEnumerator1
    Skeletal Mesh=$System.Skeletal Mesh Reader
   )
   &DynamicMaterialParameters(
    Index 0 Param 1=&FloatFromCurve(FloatCurve=DataInterface() {propertyValues {Curve {keys [<at 0 1.5>,<at 0.4 1>,<at 1 0.968>]}}})
    Index 0 Param 2=1
    Index 0 Param 3=1
    Index 0 Param 4=1
    Module.Param0WriteEnabled=true
    Module.Param1WriteEnabled=true
    Module.Param2WriteEnabled=true
    Module.Param3WriteEnabled=true
   )
   &ScaleColor(
    Scale Alpha=&FloatFromCurve(
     FloatCurve=DataInterface() {propertyValues {Curve {keys [<at 0 0 RCIM_Cubic>,<at 0.5 1>]}}}
     CurveIndex=$Engine.System.Age
     Scale Curve=&FloatFromCurve(FloatCurve=DataInterface() {propertyValues {Curve {
         defaultValue=1
         keys [{interpMode=RCIM_Cubic, tangentMode=RCTM_User, time=0.3, value=1}, {interpMode=RCIM_Cubic, tangentMode=RCTM_User, time=1, value=0}]
        }}})
    )
    Scale RGB=&MakeVectorFromLinearColorRGB(Color=&ColorFromCurve(ColorCurve=DataInterface() {propertyValues {
        AlphaCurve {keys [<at 0 1 RCIM_Cubic>]}
        BlueCurve {keys [<at 0 1 RCIM_Cubic>,<at 1 1 RCIM_Cubic>]}
        GreenCurve {keys [<at 0 1 RCIM_Cubic>,<at 1 0.5246 RCIM_Cubic>]}
        RedCurve {keys [<at 0 1 RCIM_Cubic>,<at 1 0 RCIM_Cubic>]}
       }}))
    ScaleA=true
    ScaleRGB=true
   )
   &ScaleMeshSize(Scale Factor=&VectorFromFloat(Value=&FloatFromCurve(
      FloatCurve=DataInterface() {propertyValues {Curve {keys [{interpMode=RCIM_Cubic, leaveTangent=1.6767, tangentMode=RCTM_Break, time=0, value=0}, <at 1 1.7>]}}}
      Scale Curve=&Multiply_Float(A=1.2, B=1)
     )))
   &CameraOffset(Camera Offset Amount=50)
  ]
  rendererClasses [&NiagaraMeshRendererProperties]
  renderers [{rendererClass=&NiagaraMeshRendererProperties, rendererIndex=0}]
 }
 Init_Glow001 {
  emitterUpdateScript [
   &EmitterState(Life Cycle Mode=::NewEnumerator1, Loop Behavior=::NewEnumerator1, Loop Duration Mode=::NewEnumerator1, UseLoopDelay=true)
   /Niagara/Modules/Emitter/SpawnByTrigger(
    Trigger:NiagaraBool=&SetBoolByFloatComparison(A=$System.Timeline.Time, B=0, Comparison Type=::NewEnumerator4)
    TriggerIntervalSwitch=false
   )
  ]
  particleSpawnScript [
   &InitializeParticle(
    Color=<rgba 1.3976 6.8947 10 5>
    Lifetime=&RandomRangeFloat(Minimum=0.3, Maximum=0.6)
    Mesh Scale Mode=::NewEnumerator3
    Mesh Uniform Scale=&Multiply_Float(A=0.02, B=2)
    UsePositionOffset=true
    Write Material Random=true
   ) {module=InitializeParticle001}
   &PartitionParticles()
   &InitialMeshOrientation(
    Rotation=&RandomRangeVector(Minimum=<0 0 0>, Maximum=<1 1 1>)
    Use Orientation Vector=true
    Use Rotation Vector=true
   ) {module=InitialMeshOrientation001}
   &SystemLocation(MaskSpawnGroup=false, Offset=<0 25 40>) {module=SystemLocation001}
  ]
  particleUpdateScript [
   &ParticleState()
   &SkeletalMeshLocation(
    Bone Sampling Mode=::NewEnumerator4
    MaskSpawnGroup=false
    Orientation Sampling=::NewEnumerator1
    Sampled Position Offset=&Multiply_Vector(
     A=&Add_Vector(
      A=&Lerp_Vector(
       A=&MakeVector(
        X=0
        Y=&FloatFromCurve(FloatCurve=DataInterface() {propertyValues {Curve {defaultValue=1, keys [<at 0 0.32 RCIM_Cubic>]}}}, Scale Curve=50)
        Z=&FloatFromCurve(FloatCurve=DataInterface() {propertyValues {Curve {defaultValue=1, keys [<at 0 -0.23 RCIM_Cubic>]}}}, Scale Curve=100)
       )
       B=&MakeVector(
        X=&FloatFromCurve(FloatCurve=DataInterface() {propertyValues {Curve {defaultValue=1, keys [{tangentMode=RCTM_User, time=0, value=0.3}]}}}, Scale Curve=150)
        Y=&FloatFromCurve(FloatCurve=DataInterface() {propertyValues {Curve {defaultValue=1, keys [<at 0 0.3 RCIM_Cubic>]}}}, Scale Curve=50)
        Z=&FloatFromCurve(FloatCurve=DataInterface() {propertyValues {Curve {defaultValue=1, keys [<at 0 -0.3 RCIM_Cubic>]}}}, Scale Curve=100)
       )
       Alpha=$User.ParaGlideLeanAlpha
      )
      B=<0 -15 -15>
     )
     B=&MergeHelper_VectorSwitch(0=<1 1 1>, 1=<-1 1 1>, 2=?, 3=?, Index=$Particles.PartitionParticles.Partition, Num=::NewEnumerator0)
    )
    Sampled Position Offset Coordinate Space:ENiagaraSkelMeshTransforms=::NewEnumerator1
    Skeletal Mesh=$System.Skeletal Mesh Reader
   )
   &DynamicMaterialParameters(
    Index 0 Param 1=&FloatFromCurve(FloatCurve=DataInterface() {propertyValues {Curve {keys [<at 0 1.5>,<at 0.4 1>,<at 1 0.968>]}}})
    Index 0 Param 2=1
    Index 0 Param 3=1
    Index 0 Param 4=1
    Module.Param0WriteEnabled=true
    Module.Param1WriteEnabled=true
    Module.Param2WriteEnabled=true
    Module.Param3WriteEnabled=true
   )
   &ScaleColor(
    Scale Alpha=&FloatFromCurve(
     FloatCurve=DataInterface() {propertyValues {Curve {keys [<at 0 0 RCIM_Cubic>,<at 0.5 1>]}}}
     CurveIndex=$Engine.System.Age
     Scale Curve=&FloatFromCurve(FloatCurve=DataInterface() {propertyValues {Curve {
         defaultValue=1
         keys [{interpMode=RCIM_Cubic, tangentMode=RCTM_User, time=0.3, value=1}, {interpMode=RCIM_Cubic, tangentMode=RCTM_User, time=1, value=0}]
        }}})
    )
    Scale RGB=&MakeVectorFromLinearColorRGB(Color=&ColorFromCurve(ColorCurve=DataInterface() {propertyValues {
        AlphaCurve {keys [<at 0 1 RCIM_Cubic>]}
        BlueCurve {keys [<at 0 1 RCIM_Cubic>,<at 1 1 RCIM_Cubic>]}
        GreenCurve {keys [<at 0 1 RCIM_Cubic>,<at 1 0.5246 RCIM_Cubic>]}
        RedCurve {keys [<at 0 1 RCIM_Cubic>,<at 1 0 RCIM_Cubic>]}
       }}))
    ScaleA=true
    ScaleRGB=true
   )
   &ScaleMeshSize(Scale Factor=&VectorFromFloat(Value=&FloatFromCurve(
      FloatCurve=DataInterface() {propertyValues {Curve {keys [{interpMode=RCIM_Cubic, leaveTangent=1.6767, tangentMode=RCTM_Break, time=0, value=0}, <at 1 1.7>]}}}
      Scale Curve=&Multiply_Float(A=1.2, B=1)
     )))
   &CameraOffset(Camera Offset Amount=50)
  ]
  rendererClasses [&NiagaraMeshRendererProperties]
  renderers [{rendererClass=&NiagaraMeshRendererProperties, rendererIndex=0}]
 }
 Left {
  emitterUpdateScript [
   &EmitterState()
   &SpawnRate(SpawnRate=30, Use Spawn Probability=false)
  ]
  particleSpawnScript [
   &PartitionParticles(Number Of Partitions=12, Particles Per Partition=1, Partition Mode:ENiagara_PartitionMode_v2=::NewEnumerator1)
   /SetVariables(Particles.Play:NiagaraInt32=&RandomRangeInt(Minimum=0, Maximum=8)) {isSetParameters=true}
   /SetVariables(Particles.KeyOrder:NiagaraInt32=&MakeIntFromFloat(Float=&self:Merge_Helper_Float_Switch_15(
      0:NiagaraFloat=3
      1:NiagaraFloat=0
      2:NiagaraFloat=2
      3:NiagaraFloat=0
      4:NiagaraFloat=1
      5:NiagaraFloat=3
      6:NiagaraFloat=0
      7:NiagaraFloat=2
      8:NiagaraFloat=0
      9:NiagaraFloat=2
      10:NiagaraFloat=0
      11:NiagaraFloat=1
      12:NiagaraFloat=?
      13:NiagaraFloat=?
      14:NiagaraFloat=?
      Index:NiagaraInt32=$Particles.PartitionParticles.Partition
      Num:NiagaraInt32=10
     ))) {isSetParameters=true, module=SetVariables_#2}
   /SetVariables(Particles.VisibilityTag=$Particles.KeyOrder) {isSetParameters=true, module=SetVariables_#3}
   &InitializeParticle(
    Color=&SelectLinearColorFromArray(
     Array Sampling Mode=::NewEnumerator1
     Direct Array Index=$Particles.VisibilityTag
     Color Selection Array=DataInterface() {propertyValues {ColorData [<rgba 0 0 0 1>, <rgba 1 1 1 1>]}}
     Random Seed=?
    )
    Lifetime=&Lerp_Float(A=1.5, B=1, Alpha=$User.ParaGlideLeanAlpha)
    Mesh Scale Mode=::NewEnumerator3
    Mesh Uniform Scale=0.1
    UsePositionOffset=false
    Write Material Random=true
   )
   &InitialMeshOrientation(Mesh Orientation Mode=::NewEnumerator1, Rotation Coordinate Space=::NewEnumerator2, Use Rotation Vector=true)
  ]
  particleUpdateScript [
   &ParticleState()
   &ScaleMeshSize(Scale Factor=&VectorFromFloat(Value=&FloatFromCurve(FloatCurve=DataInterface() {propertyValues {Curve {keys [<at 0 0>,<at 0.3 1>,<at 0.8167 1>,<at 1 0>]}}})))
   &SkeletalMeshLocation(
    Bone Sampling Mode=::NewEnumerator4
    MaskSpawnGroup=false
    Orientation Sampling=::NewEnumerator1
    Sampled Position Offset=&Lerp_Vector(
     A=&MakeVector(
      X=&FloatFromCurve(
       FloatCurve=DataInterface() {propertyValues {Curve {defaultValue=1, keys [{tangentMode=RCTM_User, time=0, value=0}, {tangentMode=RCTM_User, time=1, value=1}]}}}
       Scale Curve=150
      )
      Y=&FloatFromCurve(
       FloatCurve=DataInterface() {propertyValues {Curve {
          defaultValue=1
          keys [<at 0 0 RCIM_Cubic>, {arriveTangent=-0.5, interpMode=RCIM_Cubic, leaveTangent=-0.5, time=0.4064, value=1}, <at 1 -0.5 RCIM_Cubic>]
         }}}
       Scale Curve=50
      )
      Z=&FloatFromCurve(
       FloatCurve=DataInterface() {propertyValues {Curve {
          defaultValue=1
          keys [<at 0 -0.4 RCIM_Cubic>, {arriveTangent=4.6318, interpMode=RCIM_Cubic, leaveTangent=4.6318, tangentMode=RCTM_User, time=1, value=2}]
         }}}
       Scale Curve=100
      )
     )
     B=&MakeVector(
      X=&FloatFromCurve(
       FloatCurve=DataInterface() {propertyValues {Curve {defaultValue=1, keys [{tangentMode=RCTM_User, time=0, value=0.3}, {tangentMode=RCTM_User, time=1, value=1}]}}}
       Scale Curve=150
      )
      Y=&FloatFromCurve(
       FloatCurve=DataInterface() {propertyValues {Curve {
          defaultValue=1
          keys [<at 0 0 RCIM_Cubic>, {arriveTangent=-0.5, interpMode=RCIM_Cubic, leaveTangent=-0.5, time=0.4064, value=1}, <at 1 -0.5 RCIM_Cubic>]
         }}}
       Scale Curve=50
      )
      Z=&FloatFromCurve(
       FloatCurve=DataInterface() {propertyValues {Curve {
          defaultValue=1
          keys [<at 0 -0.4 RCIM_Cubic>, {arriveTangent=4.6318, interpMode=RCIM_Cubic, leaveTangent=4.6318, tangentMode=RCTM_User, time=1, value=2}]
         }}}
       Scale Curve=100
      )
     )
     Alpha=$User.ParaGlideLeanAlpha
    )
    Sampled Position Offset Coordinate Space:ENiagaraSkelMeshTransforms=::NewEnumerator2
    Skeletal Mesh=$System.Skeletal Mesh Reader
   )
   &UpdateMeshOrientation(Rotation Vector=&MakeVector(
     X=&Multiply_Float(
      A=&FloatFromCurve(
       FloatCurve=DataInterface() {propertyValues {Curve {defaultValue=1, keys [{tangentMode=RCTM_User, time=0, value=0}, {tangentMode=RCTM_User, time=1, value=1}]}}}
       Scale Curve=&Waveform(Global Phase=$Particles.Age, "Waveform [1]"=::NewEnumerator0, "[1] Frequency"=1, "[1] Amplitude Min/Max"=<-0.7 0.7>)
      )
      B=&Waveform("Waveform [1]"=::NewEnumerator0, "[1] Frequency"=1, "[1] Amplitude Min/Max"=<0 1>)
     )
     Y=&FloatFromCurve(FloatCurve=DataInterface() {propertyValues {Curve {keys [<at 0 0.4>,<at 0.9868 -0.2>]}}}, Scale Curve=-1)
     Z=0
    ))
   &BlendSpace(
    "Mesh Orientation Blend (World <-> Local)":NiagaraFloat=&FloatFromCurve(FloatCurve=DataInterface() {propertyValues {Curve {keys [<at 0.3 1>,<at 1 0>]}}})
    Mesh Orientation Mode=1
    "Position Blend (World <-> Local)"=0
    Rotation Contribution=0
    Translation Contribution=0
   )
   /SetVariables(Particles.Color=&LerpLinearColors(
     StartColor=$Particles.Initial.Color
     EndColor=&SelectLinearColorFromArray(
      Array Sampling Mode=::NewEnumerator0
      Direct Array Index=?
      Color Selection Array=DataInterface() {propertyValues {ColorData [<rgba 0 3.8645 5 1>]}}
      Random Seed=0
     )
     LerpFactor=&Lerp_Float(
      A=0
      B=&FloatFromCurve(FloatCurve=DataInterface() {propertyValues {Curve {keys [
           <at 0 0>
           {arriveTangent=3.3235, interpMode=RCIM_Cubic, leaveTangent=3.3235, time=0.2794, value=0}
           <at 0.3009 1 RCIM_Cubic>
           {arriveTangent=-0.0114, interpMode=RCIM_Cubic, leaveTangent=-0.0114, tangentMode=RCTM_User, time=0.5304, value=0}
           <at 1 0>
          ]}}})
      Alpha=&MakeCustomFloatFromBool(
       Bool=&SetBoolByIntComparison(A=$Particles.Play, B=0, Comparison Type=::NewEnumerator4)
       True Float=1
       False Float=0
      )
     )
    )) {isSetParameters=true}
  ]
  rendererClasses [&NiagaraMeshRendererProperties]
  renderers [
   {rendererClass=&NiagaraMeshRendererProperties, rendererIndex=0}
   {rendererClass=&NiagaraMeshRendererProperties, rendererIndex=1}
   {rendererClass=&NiagaraMeshRendererProperties, rendererIndex=2}
   {rendererClass=&NiagaraMeshRendererProperties, rendererIndex=3}
  ]
 }
 Notes {
  emitterUpdateScript [
   &EmitterState(
    Enable Distance Culling=true
    Enable Visibility Culling=true
    Life Cycle Mode=::NewEnumerator1
    Loop Behavior=::NewEnumerator1
    Loop Delay=0.6
    Loop Duration Mode=::NewEnumerator1
    MaxDistance=2000
    Module.bUseMaxDistance=true
    Module.bUseMinDistance=false
    Scalability Mode=::NewEnumerator1
    UseLoopDelay=true
   )
   &SpawnBurst_Instantaneous(Spawn Group=2, Use Spawn Probability=false, UseLoopCountLimit=false)
   &SpawnRate(
    Spawn Probability=&RandomRangeFloat(Minimum=0, Maximum=1, Evaluation Type=::NewEnumerator1)
    SpawnRate=8
    Use Spawn Probability=true
   ) {module=SpawnRate001}
  ]
  particleSpawnScript [
   /SetVariables(Particles.VisibilityTag=&SpawnGroup()) {isSetParameters=true}
   &InitializeParticle(
    Color=<rgba 0.0798 0.3504 1 1>
    Lifetime Max=0.7
    Lifetime Min=0.2
    Lifetime Mode=::NewEnumerator1
    Sprite Size Mode=::NewEnumerator3
    Uniform Sprite Size=&MaskFloatBySpawnGroup(
     Masked Float=&RandomRangeFloat(Minimum=12, Maximum=20)
     Passthrough Float=&RandomRangeFloat(Minimum=0.6, Maximum=2.5)
    )
    UsePositionOffset=false
    Write Material Random=true
   ) {module=InitializeParticle001}
   &SkeletalMeshLocation(
    Bone / Socket Index=&RandomRangeInt(Minimum=2, Maximum=3)
    Bone Sampling Mode=::NewEnumerator4
    MaskSpawnGroup=false
    Skeletal Mesh=$System.Skeletal Mesh Reader
   )
   &AddVelocityFromPoint(Module.Use Falloff Distance=false, Origin Offset=<0 0 -50>, Velocity Strength=50)
   &AddVelocity(
    Cone Angle=90
    Cone Axis=<0 0.8 1>
    Module.UseConeFalloff=false
    Module.UseVelDistribution=false
    Override Seed=false
    Velocity Mode=::NewEnumerator2
    Velocity Speed=&RandomRangeFloat(Minimum=227.5, Maximum=280)
   )
   &SubUVAnimation(
    Evaluation Type=::NewEnumerator0
    "Sprite Renderer "=DataInterface() {propertyValues {SpriteRenderer=&self:Notes_0.NiagaraSpriteRendererProperties_0}}
    SubUV Animation Mode:ENiagara_SubUVLookupModeV2=::NewEnumerator2
    UseEndFrame=false
    UseStartFrame=false
   )
  ]
  particleUpdateScript [
   &ParticleState()
   &GravityForce(Gravity=&RandomRangeVector(Minimum=<0 0 250>, Maximum=<0 0 325>))
   /Niagara/Modules/Update/Forces/CurlNoiseForce(
    Noise Frequency=0.04
    Noise Strength=&FloatFromCurve(
     FloatCurve=DataInterface() {propertyValues {Curve {keys [<at 0 1>,<at 1 0>]}}}
     Scale Curve=&RandomRangeFloat(Minimum=900, Maximum=1500)
    )
    Pan Noise=true
    Pan Noise Field=<0 0 5>
   )
   &Drag(Drag=&RandomRangeFloat(Minimum=5, Maximum=8), Use Linear Drag=true, Use Rotational Drag=false)
   &SolveForcesAndVelocity(Clamp Velocity=false, Limit Acceleration=false)
   &ScaleSpriteSizeBySpeed(
    Max Scale Factor=<0.7 3>
    Sample Scale Factor By Curve=true
    Scale Factor Curve=DataInterface() {propertyValues {Curve {keys [
        <at -0.0077 0.0765 RCIM_Cubic>
        {arriveTangent=1.2463, interpMode=RCIM_Cubic, leaveTangent=1.2463, tangentMode=RCTM_User, time=0.9985, value=0.9549}
       ]}}}
    Velocity Threshold=350
   )
   &ScaleSpriteSize(Uniform Curve Sprite Scale=DataInterface() {propertyValues {Curve {keys [<at 0.8 1 RCIM_Cubic>, {arriveTangent=-6.6126, interpMode=RCIM_Cubic, leaveTangent=-6.6126, tangentMode=RCTM_User, time=1, value=0}]}}})
   &ScaleColor(
    Scale Alpha=&FloatFromCurve(FloatCurve=DataInterface() {propertyValues {Curve {keys [<at 0 0>,<at 0.144 1>,<at 1 0>]}}})
    Scale RGB=&MakeVectorFromLinearColorRGB(Color=&ColorFromCurve(ColorCurve=DataInterface() {propertyValues {
        AlphaCurve {keys [<at 0 1 RCIM_Cubic>]}
        BlueCurve {keys [<at 0 15 RCIM_Cubic>,<at 0.2499 5>,<at 0.7114 1 RCIM_Cubic>]}
        GreenCurve {keys [<at 0 15 RCIM_Cubic>,<at 0.2499 5>,<at 0.7114 1 RCIM_Cubic>]}
        RedCurve {keys [<at 0 15 RCIM_Cubic>,<at 0.2499 5>,<at 0.7114 1 RCIM_Cubic>]}
       }}))
    ScaleA=true
    ScaleRGB=true
   )
   &BlendSpace(Mesh Orientation Mode=2, Rotation Contribution=0)
  ]
  rendererClasses [&NiagaraSpriteRendererProperties]
  renderers [
   {rendererClass=&NiagaraSpriteRendererProperties}
   {rendererClass=&NiagaraSpriteRendererProperties, rendererIndex=1}
  ]
 }
 Paper_Burst {
  emitterUpdateScript [
   &EmitterState(Life Cycle Mode=::NewEnumerator1, Loop Behavior=::NewEnumerator1, Loop Duration=1, UseLoopDelay=true)
   &SpawnBurst_Instantaneous(Spawn Count=7, Use Spawn Probability=false)
  ]
  particleSpawnScript [
   &InitializeParticle(
    Color=<rgba 0.6 0.3649 0.1242 0.3>
    Lifetime=&RandomRangeFloat(Minimum=1.7, Maximum=3)
    Mesh Scale=&Multiply_VectorByFloat(
     Vector=&RandomRangeVector(Minimum=<0.6 0.9 0.7>, Maximum=<0.8 0.9 0.9>)
     Float=&RandomRangeFloat(Minimum=0.11, Maximum=0.18)
    )
    Mesh Scale Mode=::NewEnumerator4
    Position Mode=::NewEnumerator0
    Write Material Random=true
   )
   &SphereLocation(
    MaskSpawnGroup=false
    Non Uniform Scale=<1 1 2>
    Override Seed=false
    Sphere Radius=5
    Surface Only=false
    UseNonUniformScale=true
    UseOffset=false
    UseSphereOrientation=false
   )
   &AddVelocityInCone(
    Cone Axis=<0 0 1>
    Cone Axis Coordinate Space=::World
    Module.Use Velocity Falloff On Cone Axis=false
    Velocity Distribution Along Cone Axis=0
    Velocity Strength=&RandomRangeFloat(Minimum=222, Maximum=666)
   )
   &InitialMeshOrientation(
    Orientation Vector=&RandomRangeVector(Minimum=<0 -0.5 -0.5>, Maximum=<0 0.5 -1>)
    Rotation Coordinate Space=::NewEnumerator0
    Use Orientation Vector=true
    Use Rotation Vector=true
   )
  ]
  particleUpdateScript [
   &Drag(
    Drag=&Lerp_Float(
     A=3
     B=2
     Alpha=&FloatFromCurve(FloatCurve=DataInterface() {propertyValues {Curve {keys [<at 0 0 RCIM_Cubic>, {interpMode=RCIM_Cubic, tangentMode=RCTM_Break, time=1, value=1}]}}})
    )
    Use Linear Drag=true
    Use Rotational Drag=false
   )
   &AccelerationForce(Acceleration=&RandomRangeVector(Minimum=<99 99 -99>, Maximum=<-99 -99 900>))
   &ParticleState()
   &UpdateMeshOrientation(
    Rotation Coordinate Space=::NewEnumerator2
    Rotation Rate=&RandomRangeFloat(Minimum=0.4, Maximum=0.9)
    Rotation Vector=&NormalizeVector(Vector To Normalize=&RandomRangeVector(Minimum=<-1 -1 -0.2>, Maximum=<1 1 0.2>))
   )
   &ScaleMeshSize(Scale Factor=&VectorFromFloat(Value=&FloatFromCurve(FloatCurve=DataInterface() {propertyValues {Curve {keys [
          <at 0 0 RCIM_Cubic>
          {arriveTangentWeight=0.034, interpMode=RCIM_Cubic, leaveTangentWeight=0.3059, tangentMode=RCTM_User, time=0.015, value=1}
          {arriveTangent=-2.7218, interpMode=RCIM_Cubic, tangentMode=RCTM_Break, time=1, value=0}
         ]}}})))
   &SolveForcesAndVelocity(Clamp Velocity=false, Limit Acceleration=false)
   &SolveRotationalForcesAndVelocity()
  ]
  rendererClasses [&NiagaraMeshRendererProperties]
  renderers [{rendererClass=&NiagaraMeshRendererProperties, rendererIndex=0}]
 }
 Paper_Rate {
  bEnabled=false
  emitterUpdateScript [
   &EmitterState(Life Cycle Mode=::NewEnumerator1, Loop Duration=1, UseLoopDelay=true)
   &SpawnRate(SpawnRate=9, Use Spawn Probability=false)
  ]
  particleSpawnScript [
   &InitializeParticle(
    Color=<rgba 0.6 0.3649 0.1242 0.3>
    Lifetime=&RandomRangeFloat(Minimum=1.7, Maximum=3)
    Mesh Scale=&Multiply_VectorByFloat(
     Vector=&RandomRangeVector(Minimum=<0.6 0.9 0.7>, Maximum=<0.8 0.9 0.9>)
     Float=&RandomRangeFloat(Minimum=0.11, Maximum=0.18)
    )
    Mesh Scale Mode=::NewEnumerator4
    Position Mode=::NewEnumerator0
    Write Material Random=true
   )
   &PartitionParticles()
   &SphereLocation(
    MaskSpawnGroup=false
    Non Uniform Scale=<1 1 2>
    Override Seed=false
    Sphere Origin:NiagaraPosition=&MaskVector(Mask Source=1, Mask Parameter=$Particles.PartitionParticles.Partition, Mask=0, Masked Vector=<0 30 40>, Passthrough Vector=<0 30 40>)
    Sphere Radius=5
    Surface Only=false
    UseNonUniformScale=true
    UseOffset=false
    UseSphereOrientation=false
   )
   &InitialMeshOrientation(
    Orientation Vector=&RandomRangeVector(Minimum=<0 -0.5 -0.5>, Maximum=<0 0.5 -1>)
    Rotation Coordinate Space=::NewEnumerator0
    Use Orientation Vector=true
    Use Rotation Vector=true
   )
  ]
  particleUpdateScript [
   &Drag(
    Drag=&Lerp_Float(
     A=3
     B=2
     Alpha=&FloatFromCurve(FloatCurve=DataInterface() {propertyValues {Curve {keys [<at 0 0 RCIM_Cubic>, {interpMode=RCIM_Cubic, tangentMode=RCTM_Break, time=1, value=1}]}}})
    )
    Use Linear Drag=true
    Use Rotational Drag=false
   )
   &AccelerationForce(Acceleration=&RandomRangeVector(Minimum=<99 99 -99>, Maximum=<-99 -99 100>))
   &ParticleState()
   &UpdateMeshOrientation(
    Rotation Coordinate Space=::NewEnumerator2
    Rotation Rate=&RandomRangeFloat(Minimum=0.4, Maximum=0.9)
    Rotation Vector=&NormalizeVector(Vector To Normalize=&RandomRangeVector(Minimum=<-1 -1 -0.2>, Maximum=<1 1 0.2>))
   )
   &ScaleMeshSize(Scale Factor=&VectorFromFloat(Value=&FloatFromCurve(FloatCurve=DataInterface() {propertyValues {Curve {keys [
          <at 0 0 RCIM_Cubic>
          {arriveTangentWeight=0.034, interpMode=RCIM_Cubic, leaveTangentWeight=0.3059, tangentMode=RCTM_User, time=0.015, value=1}
          {arriveTangent=-2.7218, interpMode=RCIM_Cubic, tangentMode=RCTM_Break, time=1, value=0}
         ]}}})))
   &SolveForcesAndVelocity(Clamp Velocity=false, Limit Acceleration=false)
   &SolveRotationalForcesAndVelocity()
  ]
  rendererClasses [&NiagaraMeshRendererProperties]
  renderers [{rendererClass=&NiagaraMeshRendererProperties, rendererIndex=0}]
 }
 Radial_Energy {
  emitterUpdateScript [
   &EmitterState(
    Inactive Response=::NewEnumerator1
    Life Cycle Mode=::NewEnumerator1
    Loop Behavior=::NewEnumerator1
    Loop Duration Mode=::NewEnumerator1
    UseLoopDelay=false
   )
   &SpawnBurst_Instantaneous(Spawn Count=2, Use Spawn Probability=false)
  ]
  particleSpawnScript [
   &InitializeParticle(Sprite Size Mode=::NewEnumerator3, Uniform Sprite Size=0.01, UsePositionOffset=false, Write Material Random=true)
   &PartitionParticles()
  ]
  particleUpdateScript [
   &ParticleState(Kill Particles When Lifetime Has Elapsed=false, Let Infinitely Lived Particles Die When Emitter Deactivates=true)
   &SkeletalMeshLocation(
    Bone Sampling Mode=::NewEnumerator4
    MaskSpawnGroup=false
    Orientation Sampling=::NewEnumerator1
    Sampled Position Offset=&Multiply_Vector(
     A=&Add_Vector(
      A=&Lerp_Vector(
       A=&MakeVector(
        X=0
        Y=&FloatFromCurve(FloatCurve=DataInterface() {propertyValues {Curve {defaultValue=1, keys [<at 0 0.32 RCIM_Cubic>]}}}, Scale Curve=50)
        Z=&FloatFromCurve(FloatCurve=DataInterface() {propertyValues {Curve {defaultValue=1, keys [<at 0 -0.23 RCIM_Cubic>]}}}, Scale Curve=100)
       )
       B=&MakeVector(
        X=&FloatFromCurve(FloatCurve=DataInterface() {propertyValues {Curve {defaultValue=1, keys [{tangentMode=RCTM_User, time=0, value=0.3}]}}}, Scale Curve=150)
        Y=&FloatFromCurve(FloatCurve=DataInterface() {propertyValues {Curve {defaultValue=1, keys [<at 0 0.3 RCIM_Cubic>]}}}, Scale Curve=50)
        Z=&FloatFromCurve(FloatCurve=DataInterface() {propertyValues {Curve {defaultValue=1, keys [<at 0 -0.3 RCIM_Cubic>]}}}, Scale Curve=100)
       )
       Alpha=$User.ParaGlideLeanAlpha
      )
      B=<0 -15 -15>
     )
     B=&MergeHelper_VectorSwitch(0=<1 1 1>, 1=<-1 1 1>, 2=?, 3=?, Index=$Particles.PartitionParticles.Partition, Num=::NewEnumerator0)
    )
    Sampled Position Offset Coordinate Space:ENiagaraSkelMeshTransforms=::NewEnumerator1
    Skeletal Mesh=$System.Skeletal Mesh Reader
   )
   &FN_Radial_Material(Base Sprite Tile V=2, Depth Fade=250, Radial Mask Power=10, Radial Mask Tile U=0, Sprite Extrusion Speed=0.6, Sprite Spin Speed=0.1)
   &Color(
    Color=&ColorFromCurve(ColorCurve=DataInterface() {propertyValues {
       AlphaCurve {keys [<at 0 1 RCIM_Cubic>,<at 1 1 RCIM_Cubic>]}
       BlueCurve {keys [<at 1 5 RCIM_Cubic>]}
       GreenCurve {keys [<at 1 2.1465 RCIM_Cubic>]}
       RedCurve {keys [<at 1 0 RCIM_Cubic>]}
      }})
    Scale Alpha=0.7
    Scale Color=&VectorFromFloat(Value=&MakeCustomFloatFromBool(Bool=$User.Hitting, True Float=2, False Float=1))
   )
   &ScaleSpriteSize(Scale Sprite Size Mode=::NewEnumerator0, Uniform Scale Factor=200) {module=ScaleSpriteSize005}
   &ScaleSpriteSize(Scale Sprite Size Mode=::NewEnumerator0, Uniform Scale Factor=&Multiply_Float(A=15, B=$User.Scale)) {module=ScaleSpriteSize004}
   &ScaleColor(
    Color Value To Scale=$Particles.Color
    Scale Alpha=&FloatFromCurve(
     FloatCurve=DataInterface() {propertyValues {Curve {
        defaultValue=1
        keys [{interpMode=RCIM_Cubic, tangentMode=RCTM_User, time=0, value=0}, {interpMode=RCIM_Cubic, tangentMode=RCTM_User, time=1, value=1}]
       }}}
     CurveIndex=$Engine.System.Age
    )
    Scale RGB=&Multiply_VectorByFloat(Vector=<1 1 1>, Float=1)
    ScaleA=true
    ScaleRGB=true
   )
   &CameraOffset(Camera Offset Amount=50)
  ]
  rendererClasses [&NiagaraSpriteRendererProperties]
  renderers [{rendererClass=&NiagaraSpriteRendererProperties}]
 }
 Right {
  emitterUpdateScript [
   &EmitterState()
   &SpawnRate(SpawnRate=30, Use Spawn Probability=false)
  ]
  particleSpawnScript [
   &PartitionParticles(Number Of Partitions=12, Particles Per Partition=1, Partition Mode:ENiagara_PartitionMode_v2=::NewEnumerator1)
   /SetVariables(Particles.Play:NiagaraInt32=&RandomRangeInt(Minimum=0, Maximum=8)) {isSetParameters=true}
   /SetVariables(Particles.KeyOrder:NiagaraInt32=&MakeIntFromFloat(Float=&self:Merge_Helper_Float_Switch_15(
      0:NiagaraFloat=3
      1:NiagaraFloat=0
      2:NiagaraFloat=2
      3:NiagaraFloat=0
      4:NiagaraFloat=1
      5:NiagaraFloat=3
      6:NiagaraFloat=0
      7:NiagaraFloat=2
      8:NiagaraFloat=0
      9:NiagaraFloat=2
      10:NiagaraFloat=0
      11:NiagaraFloat=1
      12:NiagaraFloat=?
      13:NiagaraFloat=?
      14:NiagaraFloat=?
      Index:NiagaraInt32=$Particles.PartitionParticles.Partition
      Num:NiagaraInt32=10
     ))) {isSetParameters=true, module=SetVariables_#2}
   /SetVariables(Particles.VisibilityTag=$Particles.KeyOrder) {isSetParameters=true, module=SetVariables_#3}
   &InitializeParticle(
    Color=&SelectLinearColorFromArray(
     Array Sampling Mode=::NewEnumerator1
     Direct Array Index=$Particles.VisibilityTag
     Color Selection Array=DataInterface() {propertyValues {ColorData [<rgba 0 0 0 1>, <rgba 1 1 1 1>]}}
     Random Seed=?
    )
    Lifetime=&Lerp_Float(A=1.5, B=1, Alpha=$User.ParaGlideLeanAlpha)
    Mesh Scale Mode=::NewEnumerator3
    Mesh Uniform Scale=0.1
    UsePositionOffset=false
    Write Material Random=true
   )
   &InitialMeshOrientation(Mesh Orientation Mode=::NewEnumerator1, Rotation Coordinate Space=::NewEnumerator2, Use Rotation Vector=true)
  ]
  particleUpdateScript [
   &ParticleState()
   &ScaleMeshSize(Scale Factor=&VectorFromFloat(Value=&FloatFromCurve(FloatCurve=DataInterface() {propertyValues {Curve {keys [<at 0 0>,<at 0.3 1>,<at 0.8167 1>,<at 1 0>]}}})))
   &SkeletalMeshLocation(
    Bone Sampling Mode=::NewEnumerator4
    MaskSpawnGroup=false
    Orientation Sampling=::NewEnumerator1
    Sampled Position Offset=&Lerp_Vector(
     A=&MakeVector(
      X=&FloatFromCurve(
       FloatCurve=DataInterface() {propertyValues {Curve {defaultValue=1, keys [{tangentMode=RCTM_User, time=0, value=0}, {tangentMode=RCTM_User, time=1, value=1}]}}}
       Scale Curve=-150
      )
      Y=&FloatFromCurve(
       FloatCurve=DataInterface() {propertyValues {Curve {
          defaultValue=1
          keys [<at 0 0 RCIM_Cubic>, {arriveTangent=-0.5, interpMode=RCIM_Cubic, leaveTangent=-0.5, time=0.4064, value=1}, <at 1 -0.5 RCIM_Cubic>]
         }}}
       Scale Curve=50
      )
      Z=&FloatFromCurve(
       FloatCurve=DataInterface() {propertyValues {Curve {
          defaultValue=1
          keys [<at 0 -0.4 RCIM_Cubic>, {arriveTangent=4.6318, interpMode=RCIM_Cubic, leaveTangent=4.6318, tangentMode=RCTM_User, time=1, value=2}]
         }}}
       Scale Curve=100
      )
     )
     B=&MakeVector(
      X=&FloatFromCurve(
       FloatCurve=DataInterface() {propertyValues {Curve {defaultValue=1, keys [{tangentMode=RCTM_User, time=0, value=0.3}, {tangentMode=RCTM_User, time=1, value=1}]}}}
       Scale Curve=-150
      )
      Y=&FloatFromCurve(
       FloatCurve=DataInterface() {propertyValues {Curve {
          defaultValue=1
          keys [<at 0 0 RCIM_Cubic>, {arriveTangent=-0.5, interpMode=RCIM_Cubic, leaveTangent=-0.5, time=0.4064, value=1}, <at 1 -0.5 RCIM_Cubic>]
         }}}
       Scale Curve=50
      )
      Z=&FloatFromCurve(
       FloatCurve=DataInterface() {propertyValues {Curve {
          defaultValue=1
          keys [<at 0 -0.5 RCIM_Cubic>, {arriveTangent=4.6318, interpMode=RCIM_Cubic, leaveTangent=4.6318, tangentMode=RCTM_User, time=1, value=2}]
         }}}
       Scale Curve=100
      )
     )
     Alpha=$User.ParaGlideLeanAlpha
    )
    Sampled Position Offset Coordinate Space:ENiagaraSkelMeshTransforms=::NewEnumerator2
    Skeletal Mesh=$System.Skeletal Mesh Reader
   )
   &UpdateMeshOrientation(Rotation Vector=&MakeVector(
     X=&Multiply_Float(
      A=&FloatFromCurve(
       FloatCurve=DataInterface() {propertyValues {Curve {defaultValue=1, keys [{tangentMode=RCTM_User, time=0, value=0}, {tangentMode=RCTM_User, time=1, value=1}]}}}
       Scale Curve=&Waveform(Global Phase=$Particles.Age, "Waveform [1]"=::NewEnumerator0, "[1] Frequency"=1, "[1] Amplitude Min/Max"=<-0.7 0.7>)
      )
      B=&Waveform("Waveform [1]"=::NewEnumerator0, "[1] Frequency"=1, "[1] Amplitude Min/Max"=<0 1>)
     )
     Y=&FloatFromCurve(FloatCurve=DataInterface() {propertyValues {Curve {keys [<at 0 0.4>,<at 0.9868 -0.2>]}}})
     Z=0
    ))
   &BlendSpace(
    "Mesh Orientation Blend (World <-> Local)":NiagaraFloat=&FloatFromCurve(FloatCurve=DataInterface() {propertyValues {Curve {keys [<at 0.3 1>,<at 1 0>]}}})
    Mesh Orientation Mode=1
    "Position Blend (World <-> Local)"=0
    Rotation Contribution=0
    Translation Contribution=0
   )
   /SetVariables(Particles.Color=&LerpLinearColors(
     StartColor=$Particles.Initial.Color
     EndColor=&SelectLinearColorFromArray(
      Array Sampling Mode=::NewEnumerator0
      Direct Array Index=?
      Color Selection Array=DataInterface() {propertyValues {ColorData [<rgba 0 3.8645 5 1>]}}
      Random Seed=0
     )
     LerpFactor=&Lerp_Float(
      A=0
      B=&FloatFromCurve(FloatCurve=DataInterface() {propertyValues {Curve {keys [
           <at 0 0>
           {arriveTangent=3.3235, interpMode=RCIM_Cubic, leaveTangent=3.3235, time=0.2794, value=0}
           <at 0.3009 1 RCIM_Cubic>
           {arriveTangent=-0.0114, interpMode=RCIM_Cubic, leaveTangent=-0.0114, tangentMode=RCTM_User, time=0.5304, value=0}
           <at 1 0>
          ]}}})
      Alpha=&MakeCustomFloatFromBool(
       Bool=&SetBoolByIntComparison(A=$Particles.Play, B=0, Comparison Type=::NewEnumerator4)
       True Float=1
       False Float=0
      )
     )
    )) {isSetParameters=true}
  ]
  rendererClasses [&NiagaraMeshRendererProperties]
  renderers [
   {rendererClass=&NiagaraMeshRendererProperties, rendererIndex=0}
   {rendererClass=&NiagaraMeshRendererProperties, rendererIndex=1}
   {rendererClass=&NiagaraMeshRendererProperties, rendererIndex=2}
   {rendererClass=&NiagaraMeshRendererProperties, rendererIndex=3}
  ]
 }
 Trails {
  emitterUpdateScript [
   &EmitterState()
   &SpawnRate(SpawnRate=64, Use Spawn Probability=false)
  ]
  particleSpawnScript [
   &InitializeParticle(
    Color=<rgba 0.1547 0.6129 1 0.5>
    Lifetime=1.5
    Ribbon Facing Mode=::NewEnumerator1
    Ribbon Width=100
    Ribbon Width Mode=::NewEnumerator1
    UsePositionOffset=false
    Write Material Random=true
   )
   &PartitionParticles()
  ]
  particleUpdateScript [
   &ParticleState()
   &SkeletalMeshLocation(
    Bone / Socket Index=$Particles.PartitionParticles.Partition
    Bone Sampling Mode=::NewEnumerator4
    MaskSpawnGroup=false
    Orientation Sampling=::NewEnumerator1
    Sampled Position Offset=&Multiply_Vector(
     A=&Add_Vector(
      A=&Lerp_Vector(
       A=&MakeVector(
        X=&FloatFromCurve(
         FloatCurve=DataInterface() {propertyValues {Curve {defaultValue=1, keys [{tangentMode=RCTM_User, time=0, value=0}, {tangentMode=RCTM_User, time=1, value=1}]}}}
         Scale Curve=150
        )
        Y=&FloatFromCurve(
         FloatCurve=DataInterface() {propertyValues {Curve {
            defaultValue=1
            keys [<at 0 0 RCIM_Cubic>, {arriveTangent=-0.5, interpMode=RCIM_Cubic, leaveTangent=-0.5, time=0.4064, value=1}, <at 1 -0.5 RCIM_Cubic>]
           }}}
         Scale Curve=50
        )
        Z=&FloatFromCurve(
         FloatCurve=DataInterface() {propertyValues {Curve {
            defaultValue=1
            keys [<at 0 -0.4 RCIM_Cubic>, {arriveTangent=4.6318, interpMode=RCIM_Cubic, leaveTangent=4.6318, tangentMode=RCTM_User, time=1, value=2}]
           }}}
         Scale Curve=100
        )
       )
       B=&MakeVector(
        X=&FloatFromCurve(
         FloatCurve=DataInterface() {propertyValues {Curve {defaultValue=1, keys [{tangentMode=RCTM_User, time=0, value=0.3}, {tangentMode=RCTM_User, time=1, value=1}]}}}
         Scale Curve=150
        )
        Y=&FloatFromCurve(
         FloatCurve=DataInterface() {propertyValues {Curve {
            defaultValue=1
            keys [<at 0 0 RCIM_Cubic>, {arriveTangent=-0.5, interpMode=RCIM_Cubic, leaveTangent=-0.5, time=0.4064, value=1}, <at 1 -0.5 RCIM_Cubic>]
           }}}
         Scale Curve=50
        )
        Z=&FloatFromCurve(
         FloatCurve=DataInterface() {propertyValues {Curve {
            defaultValue=1
            keys [<at 0 -0.4 RCIM_Cubic>, {arriveTangent=4.6318, interpMode=RCIM_Cubic, leaveTangent=4.6318, tangentMode=RCTM_User, time=1, value=2}]
           }}}
         Scale Curve=100
        )
       )
       Alpha=$User.ParaGlideLeanAlpha
      )
      B=<0 -15 -15>
     )
     B=&MergeHelper_VectorSwitch(0=<1 1 1>, 1=<-1 1 1>, 2=?, 3=?, Index=$Particles.PartitionParticles.Partition, Num=::NewEnumerator0)
    )
    Sampled Position Offset Coordinate Space:ENiagaraSkelMeshTransforms=::NewEnumerator2
    Skeletal Mesh=$System.Skeletal Mesh Reader
   )
   &ScaleColor(
    Scale Alpha=&FloatFromCurve(FloatCurve=DataInterface() {propertyValues {Curve {
        defaultValue=1
        keys [
         {interpMode=RCIM_Cubic, tangentMode=RCTM_User, time=0, value=1}
         {interpMode=RCIM_Cubic, tangentMode=RCTM_User, time=0.2925, value=1}
         {arriveTangent=-1.4133, interpMode=RCIM_Cubic, leaveTangent=-1.4133, time=0.7015, value=1}
         {tangentMode=RCTM_User, time=1, value=0}
        ]
       }}})
    ScaleA=true
    ScaleRGB=true
   )
   &ScaleRibbonWidth(Ribbon Width Scale=&FloatFromCurve(FloatCurve=DataInterface() {propertyValues {Curve {
        defaultValue=1
        keys [
         {interpMode=RCIM_Cubic, tangentMode=RCTM_User, time=0, value=0}
         {arriveTangent=1.2225, interpMode=RCIM_Cubic, leaveTangent=1.2225, time=0.2329, value=1}
         {interpMode=RCIM_Cubic, tangentMode=RCTM_User, time=0.818, value=1}
         {tangentMode=RCTM_User, time=1, value=0}
        ]
       }}}))
   /SetVariables(Particles.SnowWidth:NiagaraFloat=&FloatFromCurve(
     FloatCurve=DataInterface() {propertyValues {Curve {
        defaultValue=0
        keys [
         {interpMode=RCIM_Cubic, tangentMode=RCTM_User, time=0, value=0}
         {interpMode=RCIM_Cubic, tangentMode=RCTM_User, time=0.5, value=1}
         {tangentMode=RCTM_User, time=1, value=0}
        ]
       }}}
     Scale Curve=60
    )) {isSetParameters=true}
   /SetVariables(Particles.SnowColor:LinearColor=&MultiplyLinearColor(
     Linear Color=&ColorFromCurve(ColorCurve=DataInterface() {propertyValues {
        AlphaCurve {keys [
          <at 0.1031 0>
          <at 0.3183 2>
          {arriveTangent=-2.9503, interpMode=RCIM_Cubic, leaveTangent=-2.9503, time=0.9245, value=5}
          <at 0.9962 0 RCIM_Cubic>
         ]}
        BlueCurve {keys [<at 0 10 RCIM_Cubic>,<at 0.3739 1 RCIM_Cubic>]}
        GreenCurve {keys [<at 0 9.0404 RCIM_Cubic>,<at 0.3739 0.8023 RCIM_Cubic>]}
        RedCurve {keys [<at 0 4.1144 RCIM_Cubic>,<at 0.3739 0.5438 RCIM_Cubic>]}
       }})
     Scale Factor=<1 1 1 0.5>
    )) {isSetParameters=true}
   /SetVariables(Particles.RibbonTwist:NiagaraFloat=&FloatFromCurve(FloatCurve=DataInterface() {propertyValues {Curve {keys [<at 0.2 -3>,<at 1 0>]}}})) {isSetParameters=true}
  ]
  rendererClasses [&NiagaraRibbonRendererProperties]
  renderers [
   {rendererClass=&NiagaraRibbonRendererProperties}
   {rendererClass=&NiagaraRibbonRendererProperties, rendererIndex=1}
  ]
 }
}
summary {
 systemName=NS_Contrail_Ivory
 emitters [
  {emitterName=Right, rendererClasses [&NiagaraMeshRendererProperties]}
  {emitterName=Left, rendererClasses [&NiagaraMeshRendererProperties]}
  {emitterName=Notes, rendererClasses [&NiagaraSpriteRendererProperties]}
  {emitterName=Trails, rendererClasses [&NiagaraRibbonRendererProperties]}
  {emitterName=Paper_Burst, rendererClasses [&NiagaraMeshRendererProperties]}
  {bEnabled=false, emitterName=Paper_Rate, rendererClasses [&NiagaraMeshRendererProperties]}
  {emitterName=Init_Glow, rendererClasses [&NiagaraMeshRendererProperties]}
  {emitterName=Radial_Energy, rendererClasses [&NiagaraSpriteRendererProperties]}
  {emitterName=Init_Glow001, rendererClasses [&NiagaraMeshRendererProperties]}
 ]
 userVariables [
  User.bIsFrontEnd {type=NiagaraBool}
  User.bIsFrontEndPreview {type=NiagaraBool}
  User.ChargeUpLength {type=NiagaraFloat, defaultValue=0, description=""}
  User.Hitting {type=NiagaraBool}
  User.ParaGlideLeanAlpha {type=NiagaraFloat, defaultValue=0, description=""}
  User.Scale {type=NiagaraFloat, defaultValue=1.5, description=""}
 ]
}
